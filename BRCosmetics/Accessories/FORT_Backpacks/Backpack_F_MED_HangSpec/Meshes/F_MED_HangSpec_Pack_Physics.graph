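# Physics asset for the F_MED_HangSpec backpack cosmetic, giving collision and joints to its bow wings and strings.
# Tiny capsules anchor the root, pack, bow base and sub-base bones, joined by angularly free constraints.
# Each bow wing and bottom string capsule hangs off its own base bone with narrow swing and zero twist limits.
# The two backpack wings carry larger capsule-plus-box bodies with no constraints.
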PhysicsAsset /BRCosmetics/Accessories/FORT_Backpacks/Backpack_F_MED_HangSpec/Meshes/F_MED_HangSpec_Pack_Physics
bodies {
 backpack_wing_l [
  {shapeType=Capsule, length=4, radius=4, shapeName=backpack_wing_l_capsule, center=<2.6507 4.4901 -5.1688>, rotation=<rot -0.0001 0 -67.5001>}
  {
   shapeType=Box
   extentX=5.6883
   extentY=16.5841
   extentZ=36.6421
   shapeName=backpack_wing_l_box
   center=<3.5474 13.7411 -2.3432>
   rotation=<rot 0 -6.9759 -65.0892>
  }
 ]
 backpack_wing_r [
  {shapeType=Capsule, length=4, radius=4, shapeName=backpack_wing_r_capsule, center=<-3.0709 -6.6847 4.781>, rotation=<rot -0.0001 0 -61.8751>}
  {
   shapeType=Box
   extentX=5.0599
   extentY=16.5593
   extentZ=36.6682
   shapeName=backpack_wing_r_box
   center=<-3.3202 -13.7231 2.4492>
   rotation=<rot -0.0003 -6.8402 -64.7399>
  }
 ]
 base_bow_string_l_btm [{shapeType=Capsule, length=0.5, radius=0.5, shapeName=base_bow_string_l_btm_capsule}]
 base_bow_string_r_btm [{shapeType=Capsule, length=0.5, radius=0.5, shapeName=base_bow_string_r_btm_capsule}]
 base_bow_wing_l_btm [{shapeType=Capsule, length=0.5, radius=0.5, shapeName=base_bow_wing_l_btm_capsule}]
 base_bow_wing_l_top [{shapeType=Capsule, length=0.5, radius=0.5, shapeName=base_bow_wing_l_top_capsule}]
 base_bow_wing_r_btm [{shapeType=Capsule, length=0.5, radius=0.5, shapeName=base_bow_wing_r_btm_capsule}]
 base_bow_wing_r_top [{shapeType=Capsule, length=0.5, radius=0.5, shapeName=base_bow_wing_r_top_capsule}]
 bow_base [{shapeType=Capsule, length=0.5, radius=0.5, shapeName=bow_base_capsule}]
 bow_string_l_btm [{shapeType=Capsule, length=2, radius=1, shapeName=bow_string_l_btm_capsule, center=<3.1258 0.818 0.2238>, rotation=<rot 0 93.2391 -70.0177>}]
 bow_string_r_btm [{shapeType=Capsule, length=2, radius=1, shapeName=bow_string_r_btm_capsule, center=<4.8551 0.2186 0.1659>, rotation=<rot 0 -96.9793 -81.6377>}]
 bow_wing_l_btm [{
   shapeType=Capsule
   length=2.3656
   radius=1.3443
   shapeName=bow_wing_l_btm_capsule
   center=<-0.4386 -2.1355 2.8862>
   rotation=<rot -11.8514 -90 28.0312>
  }]
 bow_wing_l_top [{
   shapeType=Capsule
   length=2.6932
   radius=1.5673
   shapeName=bow_wing_l_top_capsule
   center=<-0.5232 -2.0402 -2.6565>
   rotation=<rot 26.8252 -90 -13.5062>
  }]
 bow_wing_r_btm [{
   shapeType=Capsule
   length=2.4649
   radius=1.2602
   shapeName=bow_wing_r_btm_capsule
   center=<-0.4214 2.1333 3.2698>
   rotation=<rot 21.0534 -90 26.8155>
  }]
 bow_wing_r_top [{
   shapeType=Capsule
   length=3.023
   radius=1.5728
   shapeName=bow_wing_r_top_capsule
   center=<-1.3082 2.1675 -2.6362>
   rotation=<rot -26.5933 -90 -18.7085>
  }]
 pack [{shapeType=Capsule, length=0.5, radius=0.5, shapeName=pack_capsule}]
 root [{shapeType=Capsule, length=0.5, radius=0.5, shapeName=root_capsule}]
}
constraints [
 {
  bone1Name=base_bow_string_l_btm
  bone2Name=bow_base
  swing1LimitDegrees=45
  swing1Motion=Free
  swing2LimitDegrees=45
  swing2Motion=Free
  twistLimitDegrees=45
  twistMotion=Free
 }
 {
  bone1Name=base_bow_string_r_btm
  bone2Name=bow_base
  swing1LimitDegrees=45
  swing1Motion=Free
  swing2LimitDegrees=45
  swing2Motion=Free
  twistLimitDegrees=45
  twistMotion=Free
 }
 {
  bone1Name=base_bow_wing_l_btm
  bone2Name=bow_base
  swing1LimitDegrees=45
  swing1Motion=Free
  swing2LimitDegrees=45
  swing2Motion=Free
  twistLimitDegrees=45
  twistMotion=Free
 }
 {
  bone1Name=base_bow_wing_l_top
  bone2Name=bow_base
  swing1LimitDegrees=45
  swing1Motion=Free
  swing2LimitDegrees=45
  swing2Motion=Free
  twistLimitDegrees=45
  twistMotion=Free
 }
 {
  bone1Name=base_bow_wing_r_btm
  bone2Name=bow_base
  swing1LimitDegrees=45
  swing1Motion=Free
  swing2LimitDegrees=45
  swing2Motion=Free
  twistLimitDegrees=45
  twistMotion=Free
 }
 {
  bone1Name=base_bow_wing_r_top
  bone2Name=bow_base
  swing1LimitDegrees=45
  swing1Motion=Free
  swing2LimitDegrees=45
  swing2Motion=Free
  twistLimitDegrees=45
  twistMotion=Free
 }
 {
  bone1Name=pack
  bone2Name=root
  swing1LimitDegrees=45
  swing1Motion=Free
  swing2LimitDegrees=45
  swing2Motion=Free
  twistLimitDegrees=45
  twistMotion=Free
 }
 {bone1Name=bow_wing_r_top, bone2Name=base_bow_wing_r_top, swing1LimitDegrees=10, swing2LimitDegrees=5, twistLimitDegrees=0}
 {bone1Name=bow_wing_r_btm, bone2Name=base_bow_wing_r_btm, swing1LimitDegrees=10, swing2LimitDegrees=5, twistLimitDegrees=0}
 {bone1Name=bow_wing_l_top, bone2Name=base_bow_wing_l_top, swing1LimitDegrees=10, swing2LimitDegrees=5, twistLimitDegrees=0}
 {bone1Name=bow_wing_l_btm, bone2Name=base_bow_wing_l_btm, swing1LimitDegrees=10, swing2LimitDegrees=5, twistLimitDegrees=0}
 {bone1Name=bow_string_r_btm, bone2Name=base_bow_string_r_btm, swing1LimitDegrees=15, swing2LimitDegrees=10, twistLimitDegrees=0}
 {bone1Name=bow_string_l_btm, bone2Name=base_bow_string_l_btm, swing1LimitDegrees=15, swing2LimitDegrees=10, twistLimitDegrees=0}
]
